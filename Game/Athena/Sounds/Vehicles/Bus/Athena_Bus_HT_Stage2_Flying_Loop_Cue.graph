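# Sound cue for the Athena Battle Bus Stage2 flying loop, attenuated in 3D on the BattleBus sound class.
# A quality-level switch picks one of four mixes, all with a looping jet.
# The fullest layers balloon, wind, envelope-faded crowd, and randomly delayed, modulated creaks and horn beeps.
# The second drops creaks and beeps; the last two use one optimized flying loop plus the jet.
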SoundCue /Game/Athena/Sounds/Vehicles/Bus/Athena_Bus_HT_Stage2_Flying_Loop_Cue
root=QualityLevel_0
nodes [
 {id=Delay_1, children [Random_0], properties {delayMax=7, delayMin=1}}
 {id=Delay_2, children [Random_1], properties {delayMax=20, delayMin=5}}
 {
  id=Enveloper_0
  children [WavePlayer_19]
  properties {pitchCurve {editorCurveData {keys [{time=0}]}}, volumeCurve {editorCurveData {keys [<at 0 1>,<at 90 0>]}}}
 }
 {
  id=Enveloper_1
  children [WavePlayer_9]
  properties {pitchCurve {editorCurveData {keys [{time=0}]}}, volumeCurve {editorCurveData {keys [<at 0 1>,<at 50 0>]}}}
 }
 {id=Looping_1, children [Modulator_1]}
 {id=Looping_2, children [Modulator_2]}
 {
  id=Mixer_0
  children [Looping_1, WavePlayer_6, Looping_2, Enveloper_1, WavePlayer_7, WavePlayer_13]
  properties {inputVolume [3,1.8,0.75,1.25,0.6,0.5]}
 }
 {id=Mixer_1, children [WavePlayer_14, Enveloper_0, WavePlayer_15, WavePlayer_16], properties {inputVolume [2,1.5,1,0.5]}}
 {id=Mixer_2, children [WavePlayer_2, WavePlayer_17], properties {inputVolume [1,0.5]}}
 {id=Mixer_3, children [WavePlayer_10, WavePlayer_18], properties {inputVolume [1,0.5]}}
 {id=Modulator_1, children [Delay_1], properties {pitchMax=1.2, pitchMin=0.7, volumeMax=1.2, volumeMin=0.7}}
 {id=Modulator_2, children [Delay_2], properties {pitchMax=1, pitchMin=1, volumeMax=1.2, volumeMin=0.8}}
 {id=QualityLevel_0, children [Mixer_0,Mixer_1,Mixer_2,Mixer_3]}
 {id=Random_0, children [WavePlayer_3,WavePlayer_4,WavePlayer_5], properties {weights [1,1,1]}}
 {id=Random_1, children [WavePlayer_0,WavePlayer_1,WavePlayer_8], properties {weights [1,1,1]}}
 {id=WavePlayer_0, properties {soundWaveAssetPtr=&/SFX_FlyingBus_BeepBeep_01}}
 {id=WavePlayer_1, properties {soundWaveAssetPtr=&/SFX_FlyingBus_BeepBeep_02}}
 {id=WavePlayer_10, properties {bLooping=true, soundWaveAssetPtr=&/SFX_BusFlying_opt}}
 {id=WavePlayer_13, properties {bLooping=true, soundWaveAssetPtr=&/Veh_Bus_HT_Stage1_Jet_Loop_01}}
 {id=WavePlayer_14, properties {bLooping=true, soundWaveAssetPtr=&/SFX_Bus_Flying_BalloonRelease_Loop}}
 {id=WavePlayer_15, properties {bLooping=true, soundWaveAssetPtr=&/SFX_Bus_Flying_Wind_Loop}}
 {id=WavePlayer_16, properties {bLooping=true, soundWaveAssetPtr=&Veh_Bus_HT_Stage1_Jet_Loop_01}}
 {id=WavePlayer_17, properties {bLooping=true, soundWaveAssetPtr=&Veh_Bus_HT_Stage1_Jet_Loop_01}}
 {id=WavePlayer_18, properties {bLooping=true, soundWaveAssetPtr=&Veh_Bus_HT_Stage1_Jet_Loop_01}}
 {id=WavePlayer_19, properties {bLooping=true, soundWaveAssetPtr=&/SFX_FlyingBus_Crowd_Loop}}
 {id=WavePlayer_2, properties {bLooping=true, soundWaveAssetPtr=&SFX_BusFlying_opt}}
 {id=WavePlayer_3, properties {soundWaveAssetPtr=&/SFX_Bus_Flying_Creak_03}}
 {id=WavePlayer_4, properties {soundWaveAssetPtr=&/SFX_Bus_Flying_Creak_01}}
 {id=WavePlayer_5, properties {soundWaveAssetPtr=&/SFX_Bus_Flying_Creak_02}}
 {id=WavePlayer_6, properties {bLooping=true, soundWaveAssetPtr=&SFX_Bus_Flying_BalloonRelease_Loop}}
 {id=WavePlayer_7, properties {bLooping=true, soundWaveAssetPtr=&SFX_Bus_Flying_Wind_Loop}}
 {id=WavePlayer_8, properties {soundWaveAssetPtr=&/SFX_FlyingBus_BeepBeep_03}}
 {id=WavePlayer_9, properties {bLooping=true, soundWaveAssetPtr=&SFX_FlyingBus_Crowd_Loop}}
]
properties {
 attenuationSettings=/Game/Sounds/Attenuation/Fort_Bus_Flying_Athena_Att
 maxDistance=8000
 soundClassObject=/Game/Sounds/FortSoundClassesAndModes/Fort_BattleBus
}
