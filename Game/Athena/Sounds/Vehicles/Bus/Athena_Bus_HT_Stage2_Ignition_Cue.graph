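# Battle bus sound cue for the stage 2 ignition, playing a distant horn beep followed by an engine startup, with close lobby-engine attenuation.
# A quality-level root sends its first two levels to one mixer and its last two to a mixer using the "_opt" beep wave.
# Each mixer layers the beep, weighted louder, with a Stage1 startup wave held back by a fixed one-second delay.
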SoundCue /Game/Athena/Sounds/Vehicles/Bus/Athena_Bus_HT_Stage2_Ignition_Cue
root=QualityLevel_1
nodes [
 {id=Delay_2, children [WavePlayer_0], properties {delayMax=1, delayMin=1}}
 {id=Delay_3, children [WavePlayer_3], properties {delayMax=1, delayMin=1}}
 {id=Mixer_2, children [WavePlayer_1,Delay_2], properties {inputVolume [0.6,0.4]}}
 {id=Mixer_5, children [WavePlayer_2,Delay_3], properties {inputVolume [0.6,0.4]}}
 {id=QualityLevel_1, children [Mixer_2,Mixer_2,Mixer_5,Mixer_5]}
 {id=WavePlayer_0, properties {soundWaveAssetPtr=&/Veh_Bus_HT_Stage1_Startup_01}}
 {id=WavePlayer_1, properties {soundWaveAssetPtr=&/SFX_FlyingBus_BeepBeep_03_Distant}}
 {id=WavePlayer_2, properties {soundWaveAssetPtr=&/SFX_FlyingBus_BeepBeep_03_Distant_opt}}
 {id=WavePlayer_3, properties {soundWaveAssetPtr=&Veh_Bus_HT_Stage1_Startup_01}}
]
properties {
 attenuationSettings=/Game/Sounds/Attenuation/Fort_Bus_Lobby_Engine_Athena_Close_Att
 maxDistance=12500
 soundClassObject=/Game/Sounds/FortSoundClassesAndModes/Fort_BattleBus
 volumeMultiplier=1.3125
}
