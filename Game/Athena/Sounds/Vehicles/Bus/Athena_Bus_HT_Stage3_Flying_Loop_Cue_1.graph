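# Sound cue for the stage 3 Battle Bus flying loop in Athena, routed to the Battle Bus sound class with flying-bus attenuation.
# A quality-level switch selects one of four mixes, all layering a looping jet.
# The top two add wind and an envelope-faded crowd; only the first has randomly delayed, volume-varied horn beeps.
# The lower two swap those layers for a single optimized bus loop.
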SoundCue /Game/Athena/Sounds/Vehicles/Bus/Athena_Bus_HT_Stage3_Flying_Loop_Cue_1
root=QualityLevel_0
nodes [
 {id=Delay_2, children [Random_1], properties {delayMax=20, delayMin=5}}
 {
  id=Enveloper_0
  children [WavePlayer_19]
  properties {pitchCurve {editorCurveData {keys [{time=0}]}}, volumeCurve {editorCurveData {keys [<at 0 1>,<at 90 0>]}}}
 }
 {
  id=Enveloper_1
  children [WavePlayer_9]
  properties {pitchCurve {editorCurveData {keys [{time=0}]}}, volumeCurve {editorCurveData {keys [<at 0 1>,<at 50 0>]}}}
 }
 {id=Looping_2, children [Modulator_2]}
 {id=Mixer_0, children [Looping_2, Enveloper_1, WavePlayer_7, WavePlayer_11], properties {inputVolume [0.75,1.25,0.6,0.7]}}
 {id=Mixer_1, children [Enveloper_0, WavePlayer_15, WavePlayer_12], properties {inputVolume [1.25,0.6,0.7]}}
 {id=Mixer_2, children [WavePlayer_2, WavePlayer_20], properties {inputVolume [1,0.7]}}
 {id=Mixer_3, children [WavePlayer_10, WavePlayer_21], properties {inputVolume [1,0.7]}}
 {id=Modulator_2, children [Delay_2], properties {pitchMax=1, pitchMin=1, volumeMax=1.2, volumeMin=0.8}}
 {id=QualityLevel_0, children [Mixer_0,Mixer_1,Mixer_2,Mixer_3]}
 {id=Random_1, children [WavePlayer_0,WavePlayer_1,WavePlayer_8], properties {weights [1,1,1]}}
 {id=WavePlayer_0, properties {soundWaveAssetPtr=&/SFX_FlyingBus_BeepBeep_01}}
 {id=WavePlayer_1, properties {soundWaveAssetPtr=&/SFX_FlyingBus_BeepBeep_02}}
 {id=WavePlayer_10, properties {bLooping=true, soundWaveAssetPtr=&/SFX_BusFlying_opt}}
 {id=WavePlayer_11, properties {bLooping=true, soundWaveAssetPtr=&/Veh_Bus_HT_Stage3_Jet_Loop_01}}
 {id=WavePlayer_12, properties {bLooping=true, soundWaveAssetPtr=&Veh_Bus_HT_Stage3_Jet_Loop_01}}
 {id=WavePlayer_15, properties {bLooping=true, soundWaveAssetPtr=&/SFX_Bus_Flying_Wind_Loop}}
 {id=WavePlayer_19, properties {bLooping=true, soundWaveAssetPtr=&/SFX_FlyingBus_Crowd_Loop}}
 {id=WavePlayer_2, properties {bLooping=true, soundWaveAssetPtr=&SFX_BusFlying_opt}}
 {id=WavePlayer_20, properties {bLooping=true, soundWaveAssetPtr=&Veh_Bus_HT_Stage3_Jet_Loop_01}}
 {id=WavePlayer_21, properties {bLooping=true, soundWaveAssetPtr=&Veh_Bus_HT_Stage3_Jet_Loop_01}}
 {id=WavePlayer_7, properties {bLooping=true, soundWaveAssetPtr=&SFX_Bus_Flying_Wind_Loop}}
 {id=WavePlayer_8, properties {soundWaveAssetPtr=&/SFX_FlyingBus_BeepBeep_03}}
 {id=WavePlayer_9, properties {bLooping=true, soundWaveAssetPtr=&SFX_FlyingBus_Crowd_Loop}}
]
properties {
 attenuationSettings=/Game/Sounds/Attenuation/Fort_Bus_Flying_Athena_Att
 maxDistance=8000
 soundClassObject=/Game/Sounds/FortSoundClassesAndModes/Fort_BattleBus
}
