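# Sound cue for the Athena Battle Bus ignition: a horn beep-beep followed by the engine ignition, routed through the Battle Bus sound class.
# It layers a close and a distant version, each with its own lobby bus engine attenuation, in a top mixer that sets the distant layer slightly lower.
# In each layer the beep starts immediately and the ignition follows after a fixed one-second delay, with the distant ignition also mixed down.
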SoundCue /Game/Athena/Sounds/Vehicles/Bus/Athena_Bus_Ignition_Cue
root=Mixer_3
nodes [
 {id=Attenuation_3, children [Mixer_1], properties {attenuationSettings=/Game/Sounds/Attenuation/Fort_Bus_Lobby_Engine_Athena_Distant_Att}}
 {id=Attenuation_4, children [Mixer_4], properties {attenuationSettings=/Game/Sounds/Attenuation/Fort_Bus_Lobby_Engine_Athena_Close_Att}}
 {id=Delay_0, children [WavePlayer_8], properties {delayMax=1, delayMin=1}}
 {id=Delay_1, children [WavePlayer_6], properties {delayMax=1, delayMin=1}}
 {id=Mixer_1, children [WavePlayer_3,Delay_1], properties {inputVolume [1,0.7]}}
 {id=Mixer_3, children [Attenuation_4, Attenuation_3], properties {inputVolume [1,0.7]}}
 {id=Mixer_4, children [WavePlayer_7,Delay_0], properties {inputVolume [1,1]}}
 {id=WavePlayer_3, properties {soundWaveAssetPtr=&/SFX_FlyingBus_BeepBeep_03_Distant}}
 {id=WavePlayer_6, properties {soundWaveAssetPtr=&/SFX_FlyingBus_Ignition_Distant}}
 {id=WavePlayer_7, properties {soundWaveAssetPtr=&/SFX_FlyingBus_BeepBeep_03}}
 {id=WavePlayer_8, properties {soundWaveAssetPtr=&/SFX_FlyingBus_Ignition}}
]
properties {maxDistance=50000, soundClassObject=/Game/Sounds/FortSoundClassesAndModes/Fort_BattleBus, volumeMultiplier=1}
