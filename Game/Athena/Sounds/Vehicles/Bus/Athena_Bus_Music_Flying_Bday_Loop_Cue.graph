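# Looping birthday-event trap music cue for the Athena Battle Bus while flying.
# A quality-level node picks the standard looping wave for the first two tiers and the "_opt" variant for the other two.
# An enveloper at the root eases volume down over the first few seconds, and the cue uses the bus music attenuation asset and BattleBus sound class.
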SoundCue /Game/Athena/Sounds/Vehicles/Bus/Athena_Bus_Music_Flying_Bday_Loop_Cue
root=Enveloper_0
nodes [
 {
  id=Enveloper_0
  children [QualityLevel_0]
  properties {pitchCurve {editorCurveData {keys [{time=0}]}}, volumeCurve {editorCurveData {keys [<at 0 1>,<at 5 0.7>]}}}
 }
 {id=QualityLevel_0, children [WavePlayer_1,WavePlayer_1,WavePlayer_2,WavePlayer_2]}
 {id=WavePlayer_1, properties {bLooping=true, soundWaveAssetPtr=&/Bus_Flying_BdayMusic_Trap_Loop}}
 {id=WavePlayer_2, properties {bLooping=true, soundWaveAssetPtr=&/Bus_Flying_BdayMusic_Trap_Loop_opt}}
]
properties {
 attenuationSettings=/Game/Sounds/Attenuation/Fort_Bus_Music_Athena_Att
 maxDistance=21000
 soundClassObject=/Game/Sounds/FortSoundClassesAndModes/Fort_BattleBus
 volumeMultiplier=1
 attenuationOverrides {bAttenuate=false, bSpatialize=false}
}
