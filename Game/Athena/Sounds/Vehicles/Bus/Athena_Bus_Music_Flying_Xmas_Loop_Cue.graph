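# Sound cue for the Winterfest Battle Bus flying music loop in Athena.
# A single wave player feeds the Season 23 Winterfest music loop into an Enveloper, whose volume curve eases the music down to three-quarters level over the first five seconds.
# The cue plays slightly reduced in volume through the Battle Bus sound class, with the shared bus music attenuation for spatial falloff.
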SoundCue /Game/Athena/Sounds/Vehicles/Bus/Athena_Bus_Music_Flying_Xmas_Loop_Cue
root=Enveloper_1
nodes [
 {
  id=Enveloper_1
  children [WavePlayer_0]
  properties {pitchCurve {editorCurveData {keys [{time=0}]}}, volumeCurve {editorCurveData {keys [<at 0 1>,<at 5 0.75>]}}}
 }
 {id=WavePlayer_0, properties {soundWaveAssetPtr=&/s23_Battle_Bus_Winterfest_Music_Loop}}
]
properties {
 attenuationSettings=/Game/Sounds/Attenuation/Fort_Bus_Music_Athena_Att
 maxDistance=21000
 soundClassObject=/Game/Sounds/FortSoundClassesAndModes/Fort_BattleBus
 volumeMultiplier=0.8
}
