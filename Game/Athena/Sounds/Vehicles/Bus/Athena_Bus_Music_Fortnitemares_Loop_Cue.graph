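# Sound cue that plays the looping Fortnitemares variant of the Battle Bus flying music in Athena.
# A looping wave player feeds an Enveloper node, whose volume curve fades the music to half level over the first five seconds.
# The cue uses the Battle Bus sound class, slightly reduced volume and the shared Athena bus music attenuation asset.
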SoundCue /Game/Athena/Sounds/Vehicles/Bus/Athena_Bus_Music_Fortnitemares_Loop_Cue
root=Enveloper_0
nodes [
 {
  id=Enveloper_0
  children [WavePlayer_3]
  properties {pitchCurve {editorCurveData {keys [{time=0}]}}, volumeCurve {editorCurveData {keys [<at 0 1>,<at 5 0.5>]}}}
 }
 {id=WavePlayer_3, properties {bLooping=true, soundWaveAssetPtr=&/Bus_Flying_Fortnitemares_Loop}}
]
properties {
 attenuationSettings=/Game/Sounds/Attenuation/Fort_Bus_Music_Athena_Att
 maxDistance=21000
 soundClassObject=/Game/Sounds/FortSoundClassesAndModes/Fort_BattleBus
 volumeMultiplier=0.8
 attenuationOverrides {bAttenuate=false, bSpatialize=false}
}
