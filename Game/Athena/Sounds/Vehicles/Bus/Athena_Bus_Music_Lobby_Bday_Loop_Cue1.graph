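# Sound cue for the looping birthday trap music on the Battle Bus in the Athena lobby.
# An Enveloper fades volume from full to half over ten seconds, feeding from a quality-level switch.
# The switch plays the standard looping wave for its first two levels and an "_opt" variant for the other two.
# The cue uses the Battle Bus sound class, bus lobby music attenuation and a boosted volume multiplier.
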SoundCue /Game/Athena/Sounds/Vehicles/Bus/Athena_Bus_Music_Lobby_Bday_Loop_Cue1
root=Enveloper_0
nodes [
 {
  id=Enveloper_0
  children [QualityLevel_0]
  properties {pitchCurve {editorCurveData {keys [{time=0}]}}, volumeCurve {editorCurveData {keys [{arriveTangent=-0.0914, time=0, value=1}, <at 10 0.5>]}}}
 }
 {id=QualityLevel_0, children [WavePlayer_1,WavePlayer_1,WavePlayer_2,WavePlayer_2]}
 {id=WavePlayer_1, properties {bLooping=true, soundWaveAssetPtr=&/Bus_Flying_BdayMusic_Trap_Loop}}
 {id=WavePlayer_2, properties {bLooping=true, soundWaveAssetPtr=&/Bus_Flying_BdayMusic_Trap_Loop_opt}}
]
properties {
 attenuationSettings=/Game/Sounds/Attenuation/Fort_Bus_Lobby_Music_Athena_Att
 maxDistance=2097152
 soundClassObject=/Game/Sounds/FortSoundClassesAndModes/Fort_BattleBus
 volumeMultiplier=1.25
}
