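# Sound cue for the looping Battle Bus music used in the Athena lobby, assigned to the Battle Bus sound class with lobby bus music attenuation.
# A looping wave player plays the s23 Battle Bus music and feeds an Enveloper at the root.
# The Enveloper's volume curve eases the music from full to half level over the first ten seconds, on top of a slight cue-level volume boost.
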SoundCue /Game/Athena/Sounds/Vehicles/Bus/Athena_Bus_Music_Lobby_Loop_Cue
root=Enveloper_0
nodes [
 {
  id=Enveloper_0
  children [WavePlayer_0]
  properties {pitchCurve {editorCurveData {keys [{time=0}]}}, volumeCurve {editorCurveData {keys [{arriveTangent=-0.0914, time=0, value=1}, <at 10 0.5>]}}}
 }
 {id=WavePlayer_0, properties {bLooping=true, soundWaveAssetPtr=&/s23_Battle_Bus_Music_Loop}}
]
properties {
 attenuationSettings=/Game/Sounds/Attenuation/Fort_Bus_Lobby_Music_Athena_Att
 maxDistance=2097152
 soundClassObject=/Game/Sounds/FortSoundClassesAndModes/Fort_BattleBus
 volumeMultiplier=1.25
}
