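# Sound cue that plays when the Battle Bus lobby music stops in Athena, judging by its name.
# A single wave player feeds the root with a borrowed trick-sequence fail UI sound rather than music.
# It routes through the Battle Bus sound class at reduced volume, with its own attenuation override limiting how far it carries.
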SoundCue /Game/Athena/Sounds/Vehicles/Bus/Athena_Bus_Music_Lobby_Stop_Cue
root=WavePlayer_0
nodes [{id=WavePlayer_0, properties {soundWaveAssetPtr=/Game/Sounds/Fort_UI_Sounds/Gameplay/Tricks/TrickUI_SequenceEnd_Fail_08}}]
properties {
 bOverrideAttenuation=true
 maxDistance=1512
 soundClassObject=/Game/Sounds/FortSoundClassesAndModes/Fort_BattleBus
 volumeMultiplier=0.6825
 attenuationOverrides {falloffDistance=1000, attenuationShapeExtents {x=512}}
}
