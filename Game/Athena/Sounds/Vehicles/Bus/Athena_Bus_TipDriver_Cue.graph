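# Sound cue for the battle bus in Athena that plays the Tip_The_Bus_Driver sound.
# A quality level node at the root routes all four of its inputs to the same wave player, so every quality level plays the identical wave.
# It uses the battle bus sound class with reduced volume and references the close lobby bus engine attenuation asset.
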SoundCue /Game/Athena/Sounds/Vehicles/Bus/Athena_Bus_TipDriver_Cue
root=QualityLevel_1
nodes [
 {id=QualityLevel_1, children [WavePlayer_3,WavePlayer_3,WavePlayer_3,WavePlayer_3]}
 {id=WavePlayer_3, properties {soundWaveAssetPtr=&/Tip_The_Bus_Driver}}
]
properties {
 attenuationSettings=/Game/Sounds/Attenuation/Fort_Bus_Lobby_Engine_Athena_Close_Att
 maxDistance=12500
 soundClassObject=/Game/Sounds/FortSoundClassesAndModes/Fort_BattleBus
 volumeMultiplier=0.45
 attenuationOverrides {falloffDistance=1000, attenuationShapeExtents {x=0}}
}
