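# Sound cue for the Battle Bus warp balloon flame-on sound in Athena.
# A single wave player feeds the flame-on wave straight to the output, with no randomization or modulation, at reduced volume under the Battle Bus sound class.
# It overrides concurrency to one instance, stopping the oldest, disables virtualization, and carries a very large max distance.
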SoundCue /Game/Athena/Sounds/Vehicles/Bus/Athena_Bus_Warp_Balloon_Flame_On_Cue
root=WavePlayer_0
nodes [{id=WavePlayer_0, properties {soundWaveAssetPtr=&/Bus_Warp_Balloon_Flame_On_01}}]
properties {
 bOverrideConcurrency=true
 maxDistance=2097152
 soundClassObject=/Game/Sounds/FortSoundClassesAndModes/Fort_BattleBus
 virtualizationMode=Disabled
 volumeMultiplier=0.65
 concurrencyOverrides {maxCount=1, resolutionRule=StopOldest}
}
