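# Sound cue for the finish of the Battle Bus warp lens effect in Athena, routed to the Battle Bus sound class.
# A mixer plays two layers at once: a 2D wave through an attenuation node with no settings asset assigned, and a 3D wave using the bus warp flying attenuation.
# A concurrency override allows one instance, stopping the oldest, and virtualization is disabled.
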SoundCue /Game/Athena/Sounds/Vehicles/Bus/Athena_Bus_Warp_Lens_Finish_Cue
root=Mixer_0
nodes [
 {id=Attenuation_0, children [WavePlayer_1]}
 {id=Attenuation_1, children [WavePlayer_2], properties {attenuationSettings=/Game/Sounds/Attenuation/Fort_Bus_Warp_Flying_Att}}
 {id=Mixer_0, children [Attenuation_0, Attenuation_1], properties {inputVolume [1,1]}}
 {id=WavePlayer_1, properties {soundWaveAssetPtr=&/Bus_Warp_Finish_2D}}
 {id=WavePlayer_2, properties {soundWaveAssetPtr=&/Bus_Warp_Finish_3D}}
]
properties {
 bOverrideConcurrency=true
 maxDistance=2097152
 soundClassObject=/Game/Sounds/FortSoundClassesAndModes/Fort_BattleBus
 virtualizationMode=Disabled
 volumeMultiplier=1.25
 concurrencyOverrides {maxCount=1, resolutionRule=StopOldest}
}
