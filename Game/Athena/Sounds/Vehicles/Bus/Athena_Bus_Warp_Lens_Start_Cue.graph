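# Sound cue for the Battle Bus warp lens start effect, on the Battle Bus sound class.
# A single wave player feeds the Bus_Warp_Start_2D wave through an attenuation node left at its usual settings.
# A concurrency override allows one instance at a time and stops the oldest, with virtualization disabled and a very large max distance.
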SoundCue /Game/Athena/Sounds/Vehicles/Bus/Athena_Bus_Warp_Lens_Start_Cue
root=Attenuation_2
nodes [
 {id=Attenuation_2, children [WavePlayer_0]}
 {id=WavePlayer_0, properties {soundWaveAssetPtr=&/Bus_Warp_Start_2D}}
]
properties {
 bOverrideConcurrency=true
 maxDistance=2097152
 soundClassObject=/Game/Sounds/FortSoundClassesAndModes/Fort_BattleBus
 virtualizationMode=Disabled
 volumeMultiplier=1
 concurrencyOverrides {maxCount=1, resolutionRule=StopOldest}
}
