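# Sound cue for the Battle Bus warp thruster activation sound, routed to the Battle Bus sound class.
# A single 3D wave feeds the only connected input of a three-input mixer, then passes through an attenuation node using bus warp lobby thruster settings.
# It overrides concurrency to one instance, plays when silent, and has slightly raised priority that bypasses volume scaling.
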SoundCue /Game/Athena/Sounds/Vehicles/Bus/Athena_Bus_Warp_Thruster_Activate_A_Cue
root=Attenuation_1
nodes [
 {id=Attenuation_1, children [Mixer_2], properties {attenuationSettings=/Game/Sounds/Attenuation/Fort_Bus_Warp_Lobby_Thrusters_Athena_Att}}
 {id=Mixer_2, children [null,null,WavePlayer_2], properties {inputVolume [1,1,1]}}
 {id=WavePlayer_2, properties {soundWaveAssetPtr=&/Bus_Warp_Thruster_Activate_A_3D}}
]
properties {
 bBypassVolumeScaleForPriority=true
 bOverrideConcurrency=true
 maxDistance=50000
 priority=1.2
 soundClassObject=/Game/Sounds/FortSoundClassesAndModes/Fort_BattleBus
 virtualizationMode=PlayWhenSilent
 volumeMultiplier=0.85
 concurrencyOverrides {maxCount=1}
}
