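# Sound cue for the Battle Bus warp thruster dislodge, variant B, in the Athena mode.
# A single wave player outputs the Bus_Warp_Thruster_Away_B wave at reduced volume, using the flying bus warp attenuation and the Battle Bus sound class.
# Its own concurrency override allows one instance at a time, stopping the oldest.
# Virtualization is disabled, and priority is slightly raised and bypasses volume scaling.
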SoundCue /Game/Athena/Sounds/Vehicles/Bus/Athena_Bus_Warp_Thruster_Dislodge_B_Cue
root=WavePlayer_0
nodes [{id=WavePlayer_0, properties {soundWaveAssetPtr=&/Bus_Warp_Thruster_Away_B}}]
properties {
 attenuationSettings=/Game/Sounds/Attenuation/Fort_Bus_Warp_Flying_Att
 bBypassVolumeScaleForPriority=true
 bOverrideConcurrency=true
 maxDistance=2097152
 priority=1.2
 soundClassObject=/Game/Sounds/FortSoundClassesAndModes/Fort_BattleBus
 virtualizationMode=Disabled
 volumeMultiplier=0.5
 concurrencyOverrides {maxCount=1, resolutionRule=StopOldest}
}
