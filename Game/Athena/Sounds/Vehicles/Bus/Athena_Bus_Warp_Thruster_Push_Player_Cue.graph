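# Sound cue for the Battle Bus warp thruster push on a player, played at half volume in the Battle Bus sound class.
# An IsLocal branch selects one of two mixer, modulator and random chains, with false and unset sharing the second.
# Both chains pick from the same four push waves with equal weights.
# It uses an attenuation asset, allows one concurrent instance and keeps playing when silent.
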SoundCue /Game/Athena/Sounds/Vehicles/Bus/Athena_Bus_Warp_Thruster_Push_Player_Cue
root=Branch_0
nodes [
 {id=Branch_0, children [Mixer_0,Mixer_1,Mixer_1], properties {boolParameterName=IsLocal}}
 {id=Mixer_0, children [Modulator_0], properties {inputVolume [1]}}
 {id=Mixer_1, children [Modulator_1], properties {inputVolume [1]}}
 {id=Modulator_0, children [Random_0]}
 {id=Modulator_1, children [Random_1]}
 {
  id=Random_0
  children [WavePlayer_2,WavePlayer_3,WavePlayer_4,WavePlayer_5]
  properties {bShouldExcludeFromBranchCulling=true, weights [1,1,1,1]}
 }
 {
  id=Random_1
  children [WavePlayer_0,WavePlayer_1,WavePlayer_6,WavePlayer_7]
  properties {bShouldExcludeFromBranchCulling=true, weights [1,1,1,1]}
 }
 {id=WavePlayer_0, properties {soundWaveAssetPtr=&/Bus_Warp_Thruster_Player_Push_01}}
 {id=WavePlayer_1, properties {soundWaveAssetPtr=&/Bus_Warp_Thruster_Player_Push_02}}
 {id=WavePlayer_2, properties {soundWaveAssetPtr=&Bus_Warp_Thruster_Player_Push_01}}
 {id=WavePlayer_3, properties {soundWaveAssetPtr=&Bus_Warp_Thruster_Player_Push_02}}
 {id=WavePlayer_4, properties {soundWaveAssetPtr=&/Bus_Warp_Thruster_Player_Push_03}}
 {id=WavePlayer_5, properties {soundWaveAssetPtr=&/Bus_Warp_Thruster_Player_Push_04}}
 {id=WavePlayer_6, properties {soundWaveAssetPtr=&Bus_Warp_Thruster_Player_Push_03}}
 {id=WavePlayer_7, properties {soundWaveAssetPtr=&Bus_Warp_Thruster_Player_Push_04}}
]
properties {
 attenuationSettings=&/Bus_Warp_Thuster_Push_Player_Att
 bBypassVolumeScaleForPriority=true
 bOverrideConcurrency=true
 maxDistance=2800
 priority=1.2
 soundClassObject=/Game/Sounds/FortSoundClassesAndModes/Fort_BattleBus
 virtualizationMode=PlayWhenSilent
 volumeMultiplier=0.5
 concurrencyOverrides {maxCount=1}
}
