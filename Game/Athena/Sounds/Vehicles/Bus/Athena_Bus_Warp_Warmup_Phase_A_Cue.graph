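# Sound cue for phase A of the Battle Bus warp warmup, layering a thruster activation sound with a warp horn.
# A mixer plays the thruster wave at a boosted level alongside a quieter horn, which starts after a fixed delay and has its own horn attenuation.
# It uses the Battle Bus sound class, raised priority and a single-instance concurrency override, and keeps playing when silent.
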SoundCue /Game/Athena/Sounds/Vehicles/Bus/Athena_Bus_Warp_Warmup_Phase_A_Cue
root=Mixer_0
nodes [
 {id=Attenuation_0, children [WavePlayer_0]}
 {id=Attenuation_2, children [Delay_0], properties {attenuationSettings=/Game/Sounds/Attenuation/Fort_Bus_Warp_Horn_Att}}
 {id=Delay_0, children [WavePlayer_1], properties {delayMax=1.5, delayMin=1.5}}
 {id=Mixer_0, children [Attenuation_0, Attenuation_2], properties {inputVolume [1.5,0.5]}}
 {id=WavePlayer_0, properties {soundWaveAssetPtr=&/Bus_Warp_Thruster_Activate_A_2D}}
 {id=WavePlayer_1, properties {soundWaveAssetPtr=&/Bus_Warp_Horn_01}}
]
properties {
 bBypassVolumeScaleForPriority=true
 bOverrideConcurrency=true
 maxDistance=2097152
 priority=1.2
 soundClassObject=/Game/Sounds/FortSoundClassesAndModes/Fort_BattleBus
 virtualizationMode=PlayWhenSilent
 volumeMultiplier=1.334
 concurrencyOverrides {maxCount=1}
}
