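# Sound cue for phase B of the Battle Bus warp warmup, playing a thruster activation wave on the Battle Bus sound class.
# One wave player feeds an attenuation node with no settings shown, then a mixer whose second input is empty.
# The cue boosts volume and priority, bypasses volume scaling for priority, overrides concurrency to a single instance and keeps playing when silent.
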SoundCue /Game/Athena/Sounds/Vehicles/Bus/Athena_Bus_Warp_Warmup_Phase_B_Cue
root=Mixer_0
nodes [
 {id=Attenuation_0, children [WavePlayer_1]}
 {id=Mixer_0, children [Attenuation_0, null], properties {inputVolume [1.2,0.5]}}
 {id=WavePlayer_1, properties {soundWaveAssetPtr=&/Bus_Warp_Thruster_Activate_B_2D}}
]
properties {
 bBypassVolumeScaleForPriority=true
 bOverrideConcurrency=true
 maxDistance=2097152
 priority=1.2
 soundClassObject=/Game/Sounds/FortSoundClassesAndModes/Fort_BattleBus
 virtualizationMode=PlayWhenSilent
 volumeMultiplier=1.5
 concurrencyOverrides {maxCount=1}
}
