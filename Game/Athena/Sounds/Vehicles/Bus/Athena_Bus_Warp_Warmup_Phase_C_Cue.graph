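# Sound cue for phase C of the Battle Bus warp warmup, playing a 2D thruster activation sound.
# A single wave player feeds the output directly, routed through the Battle Bus sound class.
# It overrides concurrency to allow one instance, has raised priority that bypasses volume scaling, and keeps playing when silent.
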SoundCue /Game/Athena/Sounds/Vehicles/Bus/Athena_Bus_Warp_Warmup_Phase_C_Cue
root=WavePlayer_0
nodes [{id=WavePlayer_0, properties {soundWaveAssetPtr=&/Bus_Warp_Thruster_Activate_C_2D}}]
properties {
 bBypassVolumeScaleForPriority=true
 bOverrideConcurrency=true
 maxDistance=2097152
 priority=1.2
 soundClassObject=/Game/Sounds/FortSoundClassesAndModes/Fort_BattleBus
 virtualizationMode=PlayWhenSilent
 volumeMultiplier=1
 concurrencyOverrides {maxCount=1}
}
