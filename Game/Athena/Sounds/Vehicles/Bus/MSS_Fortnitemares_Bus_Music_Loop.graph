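# Stereo MetaSound source for the looping Fortnitemares battle bus music.
# On Play, it starts the wave player and sets a gain target that InterpTo eases toward over a few seconds, ducking the music from full to half level.
# The interpolated gain multiplies both channels before output.
# The source uses bus music attenuation, the BattleBus sound class and control bus volume modulation.
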MetaSoundSource /Game/Athena/Sounds/Vehicles/Bus/MSS_Fortnitemares_Bus_Music_Loop
inputs [
 Sound {type=WaveAsset}
 UE.Source.OnPlay {type=Trigger}
]
interfaces [UE.OutputFormat.Stereo:1.0, UE.Source:1.1]
outputs [
 "UE.OutputFormat.Stereo.Audio:0" {type=Audio}
 "UE.OutputFormat.Stereo.Audio:1" {type=Audio}
]
pages [Default {
  input_defaults {Sound=&/Bus_Flying_Fortnitemares_Loop, UE.Source.OnPlay=null}
  nodes [
   n1 = UE.InterpTo:Audio(Target=Value.Value:Float(Set=^UE.Source.OnPlay) {also {defaults {Init Value=1, Target Value=0.5}}}.Output Value) {also {defaults {Interp Time=5}}}
   WavePlayer_0 = UE.Wave Player:Stereo(Wave Asset=^Sound, Play=^UE.Source.OnPlay) {also {defaults {Loop="<MetaSoundBuilderResult.SUCCEEDED: 0>"}, version [1,1]}}
   Multiply_0 = UE.Multiply:Audio by Float(AdditionalOperands=@n1.Value, PrimaryOperand=@WavePlayer_0.Out Right)
   Multiply_1 = UE.Multiply:Audio by Float(AdditionalOperands=@n1.Value, PrimaryOperand=@WavePlayer_0.Out Left)
  ]
  links [{dst="@UE.OutputFormat.Stereo.Audio:1", src=Multiply_0.Out}, {dst="@UE.OutputFormat.Stereo.Audio:0", src=Multiply_1.Out}]
 }]
properties {
 assetImportData=&self:AssetImportData
 attenuationSettings=/Game/Sounds/Attenuation/Fort_Bus_Music_Athena_Att
 bLooping=true
 duration=10000
 maxDistance=21000
 numChannels=2
 outputFormat=Stereo
 sampleRate=48000
 soundClassObject=/Game/Sounds/FortSoundClassesAndModes/Fort_BattleBus
 volume=0.8
 modulationSettings {volumeRouting=Union, volumeModulationDestination {bEnableModulation=true, modulators [/Game/Sounds/Modulation/Buses/CB_FortBattleBus_Music]}}
}
